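# Additive, unlit, two-sided Niagara particle material for a small bubble effect, faking a curved surface from UVs rather than raymarching.
# A normal derived from centered UVs is exaggerated through FlattenNormal, then drives Fresnel and a skylight environment reflection.
# The reflection adds to particle color, hue-shifted by a refined radial edge mask.
# A second Fresnel, clipped by a circular sphere mask and scaled, feeds depth-faded opacity.
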Material /Game/Effects/Fort_Effects/Effects/RaymarchingBubble/SmalBubbleMat/M_RayMarchingBubble_SmallBubble
graph {
 properties {
  bAutomaticallySetUsageInEditor=false
  bUsedWithNiagaraMeshParticles=true
  bUsedWithNiagaraRibbons=true
  bUsedWithNiagaraSprites=true
  blendMode=BLEND_Additive
  shadingModel=MSM_Unlit
  twoSided=true
  usedShadingModels=Unlit
 }
 nodes [
  n1 = VectorParameter() {parameterName=F, defaultValue=<rgba 1.71 0.05 0 1>}
  n2 = DeriveNormalZ(InXY=Subtract(A=TextureCoordinate(), B=Constant2Vector() {g=0.5, r=0.5}))
  n3 = ScalarParameter() {defaultValue=-2, parameterName=F}
  n4 = Normalize(VectorInput=Transform(None=MaterialFunctionCall(Normal=@n2, Flatness=@n3) {materialFunction=&FlattenNormal}.Result))
  n5 = ReflectionVectorWS(CustomWorldNormal=@n4)
  n6 = TextureCoordinate()
  n7 = Constant() {r=0.5}
  n8 = ScalarParameter() {defaultValue=0.4, parameterName=Rad}
  n9 = VectorParameter() {parameterName=EdgeFade3_Refine, defaultValue=<rgba 1 0 0 1>}
  n10 = MaterialFunctionCall(Texture=OneMinus(None=SphereMask(A=@n6, B=@n7, Radius=@n8) {attenuationRadius=0.5, hardnessPercent=0}), RGB=@n9.RGB, A=@n9.RGB) {materialFunction=&MF_FX_Refine}
  n11 = ParticleColor()
  Add(
   A=Multiply(A=Fresnel(ExponentIn=@n1.R, BaseReflectFractionIn=@n1.R, Normal=@n4), B=SkyLightEnvMapSample(Direction=@n5))
   B=MaterialFunctionCall(Emissive=MaterialFunctionCall(Hue Shift Percentage=@n10.Result, Texture=@n11.RGB) {materialFunction=&HueShift}.Result) {materialFunction=&MF_FX_EmissiveAdjustments}.Result
  )
  n12 = VectorParameter() {parameterName=FO, defaultValue=<rgba 4 0.005 0 1>}
  n13 = Fresnel(ExponentIn=@n12.R, BaseReflectFractionIn=@n12.R, Normal=@n4)
  n14 = SphereMask(A=TextureCoordinate(), B=Constant() {r=0.5}, Radius=@n8) {attenuationRadius=0.5}
  DepthFade(Opacity=Multiply(A=Multiply(A=@n13, B=@n14), B=ScalarParameter() {defaultValue=2, parameterName=A}))
  Multiply()
  VectorParameter() {parameterName=C, defaultValue=<rgba 0 2 2 1>}
 ]
}
